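# Physics asset for the jacket of the female medium CupidHunter player cosmetic, providing capsule bodies for secondary motion of its front, side and back panels.
# Five tiny base capsules link to spine_03 with free swing and twist, and spine_03 links likewise to a larger spine_01 capsule.
# Each base carries one offset dyn capsule with zero twist and narrow swing limits, essentially a hinge at front and sides, with slight cross-swing on the single left-named back chain.
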PhysicsAsset /BRCosmetics/Characters/Player/Female/Medium/Bodies/F_MED_CupidHunter/Meshes/F_MED_CupidHunter_Jacket_Physics_
bodies {
 base_back_Jacket_l [{shapeType=Capsule, length=0.505, radius=0.505, shapeName=base_back_Jacket_l_capsule, rotation=<rot 0 -90 0>}]
 base_front_Jacket_l [{shapeType=Capsule, length=0.505, radius=0.505, shapeName=base_front_Jacket_l_capsule, rotation=<rot 0 -90 0>}]
 base_front_Jacket_r [{shapeType=Capsule, length=0.505, radius=0.505, shapeName=base_front_Jacket_r_capsule, rotation=<rot 0 -90 0>}]
 base_side_Jacket_l [{shapeType=Capsule, length=0.505, radius=0.505, shapeName=base_side_Jacket_l_capsule, rotation=<rot 0 -90 0>}]
 base_side_Jacket_r [{shapeType=Capsule, length=0.505, radius=0.505, shapeName=base_side_Jacket_r_capsule, rotation=<rot 0 -90 0>}]
 dyn_back_Jacket_l [{shapeType=Capsule, length=5, radius=3, center=<5 0 0>, rotation=<rot 90 0 0>}]
 dyn_front_Jacket_l [{shapeType=Capsule, length=5, radius=3, center=<5 0 0>, rotation=<rot 90 0 0>}]
 dyn_front_Jacket_r [{shapeType=Capsule, length=5, radius=3, center=<5 0 0>, rotation=<rot 90 0 0>}]
 dyn_side_Jacket_l [{shapeType=Capsule, length=5, radius=3, center=<5 0 0>, rotation=<rot 90 0 0>}]
 dyn_side_Jacket_r [{shapeType=Capsule, length=5, radius=3, center=<5 0 0>, rotation=<rot 90 0 0>}]
 spine_01 [{shapeType=Capsule, length=8, radius=7, shapeName=spine_01_capsule, rotation=<rot 0 -90 0>}]
 spine_03 [{shapeType=Capsule, length=7, radius=5, shapeName=spine_03_capsule, rotation=<rot 90 0 0>}]
}
constraints [
 {
  bone1Name=spine_03
  bone2Name=spine_01
  swing1LimitDegrees=45
  swing1Motion=Free
  swing2LimitDegrees=45
  swing2Motion=Free
  twistLimitDegrees=45
  twistMotion=Free
 }
 {
  bone1Name=base_back_Jacket_l
  bone2Name=spine_03
  swing1LimitDegrees=45
  swing1Motion=Free
  swing2LimitDegrees=45
  swing2Motion=Free
  twistLimitDegrees=45
  twistMotion=Free
 }
 {
  bone1Name=base_front_Jacket_r
  bone2Name=spine_03
  swing1LimitDegrees=45
  swing1Motion=Free
  swing2LimitDegrees=45
  swing2Motion=Free
  twistLimitDegrees=45
  twistMotion=Free
 }
 {
  bone1Name=base_side_Jacket_l
  bone2Name=spine_03
  swing1LimitDegrees=45
  swing1Motion=Free
  swing2LimitDegrees=45
  swing2Motion=Free
  twistLimitDegrees=45
  twistMotion=Free
 }
 {
  bone1Name=base_side_Jacket_r
  bone2Name=spine_03
  swing1LimitDegrees=45
  swing1Motion=Free
  swing2LimitDegrees=45
  swing2Motion=Free
  twistLimitDegrees=45
  twistMotion=Free
 }
 {bone1Name=dyn_back_Jacket_l, bone2Name=base_back_Jacket_l, swing1LimitDegrees=5, swing2LimitDegrees=30, twistLimitDegrees=0}
 {bone1Name=dyn_side_Jacket_l, bone2Name=base_side_Jacket_l, swing1LimitDegrees=30, swing2LimitDegrees=0, twistLimitDegrees=0}
 {bone1Name=dyn_side_Jacket_r, bone2Name=base_side_Jacket_r, swing1LimitDegrees=30, swing2LimitDegrees=0, twistLimitDegrees=0}
 {
  bone1Name=base_front_Jacket_l
  bone2Name=spine_03
  swing1LimitDegrees=45
  swing1Motion=Free
  swing2LimitDegrees=45
  swing2Motion=Free
  twistLimitDegrees=45
  twistMotion=Free
 }
 {bone1Name=dyn_front_Jacket_l, bone2Name=base_front_Jacket_l, swing1LimitDegrees=30, swing2LimitDegrees=0, twistLimitDegrees=0}
 {bone1Name=dyn_front_Jacket_r, bone2Name=base_front_Jacket_r, swing1LimitDegrees=30, swing2LimitDegrees=0, twistLimitDegrees=0}
]
