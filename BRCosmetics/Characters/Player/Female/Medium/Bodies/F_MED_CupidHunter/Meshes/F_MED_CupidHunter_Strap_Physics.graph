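# Physics asset for the strap on the medium female Cupid Hunter cosmetic body, defining its simulated shapes and joints.
# A four-link capsule chain hangs from base_Strap, with tight swing and twist limits that loosen toward the larger tip capsule.
# Pelvis and both thighs carry larger body capsules, likely as collision volumes for the strap.
# Thighs and base_Strap join the pelvis through constraints with free swing and twist.
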PhysicsAsset /BRCosmetics/Characters/Player/Female/Medium/Bodies/F_MED_CupidHunter/Meshes/F_MED_CupidHunter_Strap_Physics
bodies {
 base_Strap [{shapeType=Capsule, length=0.505, radius=1, shapeName=base_Strap_capsule, rotation=<rot 0 -90 0>}]
 dyn_Strap_01 [{shapeType=Capsule, length=2, radius=1, center=<1.964 0 0.2044>, rotation=<rot 79.9999 -179.9998 -179.9998>}]
 dyn_Strap_02 [{shapeType=Capsule, length=2, radius=1, center=<2 0 0.5988>, rotation=<rot 90 0 0>}]
 dyn_Strap_03 [{shapeType=Capsule, length=2, radius=1, center=<2 0 0.5338>, rotation=<rot 90 0 0>}]
 dyn_Strap_04 [{shapeType=Capsule, length=4, radius=2, center=<4 0 0.5338>, rotation=<rot 90 0 0>}]
 pelvis [{shapeType=Capsule, length=9, radius=9, center=<0 -0.3413 0>}]
 thigh_l [{shapeType=Capsule, length=14.5, radius=9, shapeName=thigh_l_capsule, center=<-11.4251 -1.5877 1.7936>, rotation=<rot 90 179.995 179.9953>}]
 thigh_r [{shapeType=Capsule, length=10, radius=9, shapeName=thigh_l_capsule, center=<10.6836 0.7616 1.535>, rotation=<rot 90 180 0.0003>}]
}
constraints [
 {
  bone1Name=thigh_l
  bone2Name=pelvis
  swing1LimitDegrees=45
  swing1Motion=Free
  swing2LimitDegrees=45
  swing2Motion=Free
  twistLimitDegrees=45
  twistMotion=Free
 }
 {
  bone1Name=base_Strap
  bone2Name=pelvis
  swing1LimitDegrees=45
  swing1Motion=Free
  swing2LimitDegrees=45
  swing2Motion=Free
  twistLimitDegrees=45
  twistMotion=Free
 }
 {bone1Name=dyn_Strap_01, bone2Name=base_Strap, swing1LimitDegrees=5, swing2LimitDegrees=2, twistLimitDegrees=2}
 {bone1Name=dyn_Strap_02, bone2Name=dyn_Strap_01, swing1LimitDegrees=5, swing2LimitDegrees=2, twistLimitDegrees=2}
 {bone1Name=dyn_Strap_03, bone2Name=dyn_Strap_02, swing1LimitDegrees=10, swing2LimitDegrees=5, twistLimitDegrees=3}
 {bone1Name=dyn_Strap_04, bone2Name=dyn_Strap_03, swing1LimitDegrees=15, swing2LimitDegrees=8, twistLimitDegrees=3}
 {
  bone1Name=thigh_r
  bone2Name=pelvis
  swing1LimitDegrees=45
  swing1Motion=Free
  swing2LimitDegrees=45
  swing2Motion=Free
  twistLimitDegrees=45
  twistMotion=Free
 }
]
