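# Physics asset for the female medium CupidHunter face accessory mesh, setting up physics for left and right hair braids.
# Each braid is a four-capsule dyn_braid chain hanging from a base body linked to neck_01, with swing and twist limits widening toward the tip.
# Capsules on upper spine, neck, clavicles and upper arms form a torso proxy, joined by free-rotation constraints.
# The right braid's second joint has asymmetric swing limits, unlike the left.
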PhysicsAsset /BRCosmetics/Characters/Player/Female/Medium/Bodies/F_MED_CupidHunter/Meshes/Parts/F_MED_CupidHunter_FaceAcc_Physics
bodies {
 base_braid_l [{shapeType=Capsule, length=0.505, radius=0.505, shapeName=base_braid_l_capsule, rotation=<rot 0 -90 0>}]
 base_braid_r [{shapeType=Capsule, length=0.505, radius=0.505, shapeName=base_braid_r_capsule, rotation=<rot 0 -90 0>}]
 clavicle_l [{shapeType=Capsule, length=5, radius=5, shapeName=clavicle_l_capsule, center=<5 0.2852 -0.0182>, rotation=<rot 90 0 0>}]
 clavicle_r [{shapeType=Capsule, length=5, radius=5, shapeName=clavicle_r_capsule, center=<-5 -0.0402 0>, rotation=<rot 90 0 0>}]
 dyn_braid_l_01 [{shapeType=Capsule, length=1, radius=2, center=<2 0 0>, rotation=<rot 90 0 0>}]
 dyn_braid_l_02 [{shapeType=Capsule, length=1, radius=2, center=<2 0 0>, rotation=<rot 90 0 0>}]
 dyn_braid_l_03 [{shapeType=Capsule, length=1, radius=2, center=<2 0 0>, rotation=<rot 90 0 0>}]
 dyn_braid_l_04 [{shapeType=Capsule, length=1, radius=3, center=<4 0 0>, rotation=<rot 90 0 0>}]
 dyn_braid_r_01 [{shapeType=Capsule, length=1, radius=2, center=<2 0 0>, rotation=<rot 90 0 0>}]
 dyn_braid_r_02 [{shapeType=Capsule, length=1, radius=2, center=<2 0 0>, rotation=<rot 90 0 0>}]
 dyn_braid_r_03 [{shapeType=Capsule, length=1, radius=2, center=<2 0 0>, rotation=<rot 90 0 0>}]
 dyn_braid_r_04 [{shapeType=Capsule, length=1, radius=3, center=<4 0 0>, rotation=<rot 90 0 0>}]
 neck_01 [{shapeType=Capsule, length=3.5, radius=4.5, shapeName=neck_01_capsule, center=<4.7757 0 0>, rotation=<rot 90 0 0>}]
 spine_05 [
  {shapeType=Capsule, length=15, radius=9, shapeName=spine_05_capsule, center=<-1.9437 0.3009 0>}
  {shapeType=Capsule, length=30, radius=6, shapeName=spine_05_capsule, center=<8.1682 -3.1036 0>}
 ]
 upperarm_l [{shapeType=Capsule, length=10, radius=6, shapeName=upperarm_l_capsule, center=<10.995 0 0>, rotation=<rot 90 0 0>}]
 upperarm_r [{shapeType=Capsule, length=10, radius=6, shapeName=upperarm_l_capsule, center=<-10.995 0 0>, rotation=<rot 90 0 180>}]
}
constraints [
 {
  bone1Name=base_braid_r
  bone2Name=neck_01
  swing1LimitDegrees=45
  swing1Motion=Free
  swing2LimitDegrees=45
  swing2Motion=Free
  twistLimitDegrees=45
  twistMotion=Free
 }
 {
  bone1Name=base_braid_l
  bone2Name=neck_01
  swing1LimitDegrees=45
  swing1Motion=Free
  swing2LimitDegrees=45
  swing2Motion=Free
  twistLimitDegrees=45
  twistMotion=Free
 }
 {bone1Name=dyn_braid_l_01, bone2Name=base_braid_l, swing1LimitDegrees=18, swing2LimitDegrees=18, twistLimitDegrees=5}
 {bone1Name=dyn_braid_l_02, bone2Name=dyn_braid_l_01, swing1LimitDegrees=18, swing2LimitDegrees=18, twistLimitDegrees=5}
 {bone1Name=dyn_braid_l_03, bone2Name=dyn_braid_l_02, swing1LimitDegrees=35, swing2LimitDegrees=35, twistLimitDegrees=10}
 {bone1Name=dyn_braid_l_04, bone2Name=dyn_braid_l_03, swing1LimitDegrees=35, swing2LimitDegrees=35, twistLimitDegrees=10}
 {bone1Name=dyn_braid_r_01, bone2Name=base_braid_r, swing1LimitDegrees=18, swing2LimitDegrees=18, twistLimitDegrees=5}
 {bone1Name=dyn_braid_r_02, bone2Name=dyn_braid_r_01, swing1LimitDegrees=15, swing2LimitDegrees=30, twistLimitDegrees=5}
 {bone1Name=dyn_braid_r_03, bone2Name=dyn_braid_r_02, swing1LimitDegrees=35, swing2LimitDegrees=35, twistLimitDegrees=10}
 {bone1Name=dyn_braid_r_04, bone2Name=dyn_braid_r_03, swing1LimitDegrees=35, swing2LimitDegrees=35, twistLimitDegrees=10}
 {
  bone1Name=neck_01
  bone2Name=spine_05
  swing1LimitDegrees=45
  swing1Motion=Free
  swing2LimitDegrees=45
  swing2Motion=Free
  twistLimitDegrees=45
  twistMotion=Free
 }
 {
  bone1Name=clavicle_l
  bone2Name=spine_05
  swing1LimitDegrees=45
  swing1Motion=Free
  swing2LimitDegrees=45
  swing2Motion=Free
  twistLimitDegrees=45
  twistMotion=Free
 }
 {
  bone1Name=clavicle_r
  bone2Name=spine_05
  swing1LimitDegrees=45
  swing1Motion=Free
  swing2LimitDegrees=45
  swing2Motion=Free
  twistLimitDegrees=45
  twistMotion=Free
 }
 {
  bone1Name=upperarm_l
  bone2Name=clavicle_l
  swing1LimitDegrees=45
  swing1Motion=Free
  swing2LimitDegrees=45
  swing2Motion=Free
  twistLimitDegrees=45
  twistMotion=Free
 }
 {
  bone1Name=upperarm_r
  bone2Name=clavicle_r
  swing1LimitDegrees=45
  swing1Motion=Free
  swing2LimitDegrees=45
  swing2Motion=Free
  twistLimitDegrees=45
  twistMotion=Free
 }
]
